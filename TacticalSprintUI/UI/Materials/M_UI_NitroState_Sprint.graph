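# Translucent UI material for a sprint "nitro" progress bar, with an animated gradient fill and an angled marker at the progress edge.
# A hardened UV threshold from Progress masks fill against backing for color and alpha.
# Two multiplied panning noise textures form a shimmer term, apparently fed via reroute into the three-color gradient blend.
# A rotated edge mask blends in the marker color, while static switches select simple progress or toggle backing.
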Material /TacticalSprintUI/UI/Materials/M_UI_NitroState_Sprint
graph {
 properties {blendMode=BLEND_Translucent, materialDomain=MD_UI, pixelDepthOffsetMode=PDOM_Legacy, usedShadingModels=Default Lit}
 nodes [
  n1 = MaterialFunctionCall() {materialFunction=&GetUserInterfaceUV}
  n2 = Add(A=Subtract(A=OneMinus(None=@n1.Normalized UV)), B=ScalarParameter() {defaultValue=1, group=Bar, parameterName=Progress})
  n3 = Saturate(None=Multiply(A=ComponentMask(None=@n2) {r=true}, B=ScalarParameter() {defaultValue=1024, group=Bar, parameterName=Progress_Edge_Hardness}))
  LinearInterpolate(
   A=ScalarParameter() {defaultValue=0.5, group=Bar, parameterName=Progress_Backing_Alpha}
   B=ScalarParameter() {defaultValue=1, group=Bar, parameterName=Progress_Fill_Alpha}
   Alpha=@n3
  )
  n4 = Time()
  n5 = Constant() {r=1}
  n6 = TextureSample(UVs=Panner(
    Coordinate=TextureCoordinate() {vTiling=0.1}
    Time=Multiply(A=@n4, B=@n5)
    Speed=VectorParameter() {parameterName=Speed1, defaultValue=<rgba 0.2511 0.2 0 1>}.RGB
   ) {speedX=0.3, speedY=0.8}) {
   samplerSource=SSM_Wrap_WorldGroupSettings
   samplerType=SAMPLERTYPE_LinearGrayscale
   texture=/Game/Characters/Player/Female/Medium/Bodies/F_MED_SuperNovaTaro/Textures/LowResBlurredNoise
  }
  n7 = Panner(
   Coordinate=TextureCoordinate() {uTiling=2, vTiling=0.25}
   Time=Multiply(A=Time(), B=Constant() {r=0.6})
   Speed=VectorParameter() {parameterName=Speed2, defaultValue=<rgba 0.2125 0.2 0 1>}.RGB
  ) {speedX=0.2, speedY=0.6}
  n8 = Power(Base=Multiply(A=@n6.RGB, B=TextureSample(UVs=@n7) {samplerSource=SSM_Wrap_WorldGroupSettings, texture=/Game/Athena/HUD/Phase/T_LowRes_Noise_Phase}.RGB)) {constExponent=1.8}
  n9 = Constant() {r=8}
  NamedRerouteDeclaration(None=Saturate(None=Multiply(A=@n8, B=@n9))) {name=opacity, nodeColor=<rgba 0 0 0 1>}
  n10 = LinearInterpolate(
   A=VectorParameter() {group=Color, parameterName=Backing_Color, defaultValue=<rgba 0 0 0 1>}.RGB
   B=VectorParameter() {group=Color, parameterName=Backing_Color_Stop, defaultValue=<rgba 0 0 0 1>}.RGB
   Alpha=ScalarParameter() {parameterName=AnimateBackingColor}
  )
  n11 = ComponentMask(None=@n1.Normalized UV) {r=true}
  n12 = Saturate(None=Multiply(A=OneMinus(None=@n11), B=ScalarParameter() {defaultValue=1, parameterName=3_Color_Gradient_Multiplier}))
  n13 = ScalarParameter() {defaultValue=6, parameterName=3_Color_Gradient_ChangeRate}
  n14 = MaterialFunctionCall(
   A=VectorParameter() {group=Color, parameterName=BackgroundColor, defaultValue=<rgba 0.1356 0.0242 0.0123 1>}.RGB
   B=VectorParameter() {group=Color, parameterName=MediumColor, defaultValue=<rgba 1 0.0924 0 1>}.RGB
   C=VectorParameter() {group=Color, parameterName=BrightColor, defaultValue=<rgba 1 0.688 0.026 1>}.RGB
   Alpha=Add(A=Power(Base=@n12, Exp=@n13), B=NamedRerouteUsage())
  ) {materialFunction=&3ColorBlend}
  n15 = VectorParameter() {group=Color, parameterName=Marker_Color, defaultValue=<rgba 1 1 0 1>}
  n16 = ScalarParameter() {parameterName=LiftBlendFadeStart}
  n17 = ComponentMask(None=Add(A=TextureCoordinate(), B=ScalarParameter() {parameterName=OffsetBlendFadeStart})) {r=true}
  n18 = Saturate(None=DotProduct(A=LinearInterpolate(A=@n16, Alpha=@n17), B=ScalarParameter() {group=Marker, parameterName=BlendMarkerColor}))
  n19 = ScalarParameter() {defaultValue=-0.0125, group=Marker, parameterName=MarkerAngle_Rotation}
  n20 = Subtract(
   A=OneMinus(None=MaterialFunctionCall(UVs=@n1.Normalized UV, "Rotation Angle (0-1)"=@n19) {materialFunction=&CustomRotator}.Rotated Values)
   B=Add(A=Constant() {r=1}, B=ScalarParameter() {defaultValue=0.0625, group=Marker, parameterName=MarkerAngle_Offset})
  )
  n21 = ScalarParameter() {defaultValue=1, group=Bar, parameterName=Progress}
  n22 = Multiply(A=ComponentMask(None=Add(A=@n20, B=@n21)) {r=true}, B=ScalarParameter() {defaultValue=80, group=Marker, parameterName=Angle_Edge_Hardness})
  n23 = LinearInterpolate(
   A=@n14.Result
   B=LinearInterpolate(A=LinearInterpolate(A=@n14.Result, B=@n15.RGB, Alpha=@n18), B=@n15.RGB, Alpha=ScalarParameter() {parameterName=AnimateMarkerIntensity})
   Alpha=OneMinus(None=Saturate(None=@n22))
  )
  StaticSwitchParameter(
   True=LinearInterpolate(A=@n10, B=@n14.Result, Alpha=@n3)
   False=StaticSwitchParameter(True=LinearInterpolate(A=@n10, B=@n23, Alpha=@n3), False=@n23) {parameterName=HasBacking}
  ) {parameterName=SimpleProgress}
 ]
}
