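# UI material for the tactical sprint energy bar in its unlimited state, drawn as a translucent rounded bar.
# A horizontal two-color gradient plus a time-scrolled reflected-gradient shine feeds both fill and background colors of MF_RadialProgBar.
# That function's opacity is multiplied by an optional pulse, switched by EnablePulsing and apparently driven by a looping curve-atlas glow animation.
# A box distance field is also declared but appears unused.
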Material /TacticalSprintUI/UI/Materials/M_UI_Rounded_HitPointBar_EnergyUnlimited
graph {
 properties {bAutomaticallySetUsageInEditor=false, blendMode=BLEND_Translucent, materialDomain=MD_UI, usedShadingModels=Default Lit}
 nodes [
  n1 = NamedRerouteUsage()
  n2 = ScalarParameter() {defaultValue=1.5, parameterName=ShineOpacity}
  n3 = NamedRerouteUsage()
  n4 = NamedRerouteUsage()
  n5 = MaterialFunctionCall() {materialFunction=&LinearGradient}
  n6 = Saturate(None=Add(A=Multiply(A=@n1, B=@n2), B=LinearInterpolate(A=@n3, B=@n4, Alpha=@n5.UGradient)))
  Multiply(
   A=MaterialFunctionCall!(FillColor=@n6, BGColor=@n6) {materialFunction=/Game/Athena/HUD/Bacchus/Material/MF_RadialProgBar}.Opacity
   B=LinearInterpolate(A=Constant() {r=1}, B=NamedRerouteUsage(), Alpha=ScalarParameter() {parameterName=EnablePulsing}) {constA=1, constAlpha=1}
  )
  n7 = Time() {period=5}
  n8 = Multiply(A=Frac(None=Multiply(A=@n7) {constB=0.3})) {constB=10}
  n9 = TextureCoordinate() {coordinateIndex=4, uTiling=-1, vTiling=0.4}
  n10 = ComponentMask(None=@n9) {r=true}
  n11 = MaterialFunctionCall!(UVs=AppendVector(A=Add(A=@n8, B=@n10), B=ComponentMask(None=@n9) {g=true}), Offset, RotationCenter, RotationAngle=Constant()) {materialFunction=/FortUILibrary/MaterialFunctions/Gradients/MF_UI_ReflectedGradient}
  n12 = Constant() {r=0.5}
  NamedRerouteDeclaration(None=Saturate(None=Subtract(A=@n11.Result, B=@n12))) {name=Shine, nodeColor=<rgba 0 0 0 1>}
  n13 = TextureCoordinate()
  n14 = Subtract(A=Abs(None=ConstantBiasScale(None=@n13) {bias=-0.5, scale=2}), B=Constant2Vector() {g=1, r=1})
  n15 = Max(A=@n14) {constB=0}
  n16 = MaterialFunctionCall(Float2=@n14) {materialFunction=&BreakOutFloat2Components}
  n17 = Min(A=Max(A=@n16.R, B=@n16.R) {constB=0}) {constB=0}
  NamedRerouteDeclaration(None=Length(None=Add(A=@n15, B=@n17)) {is=class MaterialExpressionLength}) {name=Box, nodeColor=<rgba 0 0 0 1>}
  n18 = CurveAtlasRowParameter(CurveTime=Frac(None=Time())) {
   atlas=/TacticalSprintUI/UI/Materials/CurveAtlas/CA_EnergyDrinkAnimationAtlas_TacticalSprint
   curve=/TacticalSprintUI/UI/Materials/CurveAtlas/C_EnergyDrinkGlow_TacticalSprint
   parameterName=FlashAnimation
  }
  NamedRerouteDeclaration(None=@n18) {name=Pulsing, nodeColor=<rgba 0 0 0 1>}
  NamedRerouteDeclaration(None=VectorParameter() {group=Color, parameterName=Progress_Color_Start, defaultValue=<rgba 0.9216 0.4015 0.0619 1>}.RGB) {name=Color1, nodeColor=<rgba 0 0 0 1>}
  NamedRerouteDeclaration(None=VectorParameter() {group=Color, parameterName=Progress_Color_Stop, defaultValue=<rgba 0.9473 0.3527 0.1751 1>}.RGB) {name=Color2, nodeColor=<rgba 0 0 0 1>}
 ]
}
