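# Translucent UI master material for the tactical sprint icon, tinting a sprinter and speed lines from one packed mask texture.
# Nested lerps blend in the speed-line color by the red channel, with the green channel restoring sprinter color on top.
# Opacity sums both channels scaled by their alpha parameters, then multiplies by doubled red, so zero-red pixels vanish even where green is set.
# A final alpha parameter scales the clamped result.
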Material /TacticalSprintUI/UI/Materials/M_UI_TacSprintIcon_Master
graph {
 properties {blendMode=BLEND_Translucent, materialDomain=MD_UI, pixelDepthOffsetMode=PDOM_Legacy, usedShadingModels=Default Lit}
 nodes [
  n1 = VectorParameter() {group=Color, parameterName=Sprinter_Color, defaultValue=<rgba 1 0 0 1>}
  n2 = TextureSample(UVs=TextureCoordinate()) {samplerType=SAMPLERTYPE_LinearColor, texture=/TacticalSprintUI/UI/Textures/T_UI_TacSprint_64x}
  LinearInterpolate(
   A=LinearInterpolate(A=@n1.RGB, B=VectorParameter() {group=Color, parameterName=SpeedLines_Color, defaultValue=<rgba 0 1 0 1>}.RGB, Alpha=@n2.R)
   B=@n1.RGB
   Alpha=@n2.G
  )
  n3 = Multiply(A=@n2.R, B=ScalarParameter() {defaultValue=0.5, group=Alpha, parameterName=SpeedLines_Alpha})
  n4 = Multiply(A=@n2.G, B=ScalarParameter() {defaultValue=1, group=Alpha, parameterName=Sprinter_Alpha})
  n5 = DotProduct(A=Add(A=@n2.R, B=@n2.R), B=Saturate(None=Add(A=@n3, B=@n4)))
  Saturate(None=Multiply(A=Saturate(None=@n5), B=ScalarParameter() {defaultValue=1, group=Alpha, parameterName=Final_Alpha}))
 ]
}
