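; Dynamic UI director that adds or removes the tactical sprint stamina scene for the owning local player.
; It re-evaluates at BeginPlay and whenever the HUD element visibility refresh delegate fires, unbinding at EndPlay.
; The scene is withheld when the minigame's HUD element visibility mutator supplies a player info widget override class.
; A cached bSceneAdded flag limits add and remove calls to actual state changes.
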(event EventBeginPlay
  (CallFunction|UpdateTacticalSprintScene)
  (bind _asfort_player_controller_gameplay (Utilities|Casting|CastToFortPlayerControllerGameplay (DynamicUIDirector|GetOwningLocalPlayerController)))
  (HUD|BindEventtoOnHUDElementVisibilityRefreshedDelegate _asfort_player_controller_gameplay (EventDispatchers|CreateEvent :!event UpdateTacticalSprintScene)))

(event EventEndPlay
  (params (EndPlayReason :type EEndPlayReason%20Enum :default "Destroyed"))
  (bind _asfort_player_controller_gameplay (Utilities|Casting|CastToFortPlayerControllerGameplay (DynamicUIDirector|GetOwningLocalPlayerController)))
  (HUD|UnbindEventfromOnHUDElementVisibilityRefreshedDelegate _asfort_player_controller_gameplay (EventDispatchers|CreateEvent :!event UpdateTacticalSprintScene)))

(fn DoesPlayerInfoWidgetOverrideExist
  (Utilities|FlowControl|Sequence
    (:then_0
      (bind _asfort_player_controller_gameplay (Utilities|Casting|CastToFortPlayerControllerGameplay (DynamicUIDirector|GetOwningLocalPlayerController)))
      (Utilities|IsValid (Creative|GetMinigame _asfort_player_controller_gameplay)
        (:Is%20Valid
          (Utilities|IsValid (Actor|GetComponentbyClass (Creative|GetMinigame _asfort_player_controller_gameplay) "/Script/FortniteGame.FortMutatorListComponent")
            (:Is%20Valid
              (bind _asfort_athena_mutator_hudelement_visibility (Utilities|Casting|CastToFortAthenaMutator_HUDElementVisibility (Modifier|GetMutatorbyClass (Actor|GetComponentbyClass (Creative|GetMinigame _asfort_player_controller_gameplay) "/Script/FortniteGame.FortMutatorListComponent") "/Script/FortniteGame.FortAthenaMutator_HUDElementVisibility")))
              (bind _returnvalue (Class|FortAthenaMutatorHUDElementVisibility|GetPlayerInfoWidgetOverrideClass _asfort_athena_mutator_hudelement_visibility (DynamicUIDirector|GetOwningLocalPlayerController)))
              (return (Utilities|IsValidSoftClassReference _returnvalue)))))))
    (:then_1
      (return false))))

(fn UpdateTacticalSprintScene
  ; Don't add scene if an override exists
  (bind _bdoesplayerinfowidgetoverrideexist (Default|DoesPlayerInfoWidgetOverrideExist))
  (LocalVariables|SetbShouldAddScene (not _bdoesplayerinfowidgetoverrideexist))
  (if (!= (Variables|Default|GetbSceneAdded) (LocalVariables|GetbShouldAddScene))
    (bind _output_get (Variables|Default|SetbSceneAdded (LocalVariables|GetbShouldAddScene)))
    (if _output_get
      (DynamicUIDirector|AddScene "/TacticalSprintUI/UI/TacSprintStamScene.TacSprintStamScene")
      (else
        (DynamicUIDirector|RemoveScene "/TacticalSprintUI/UI/TacSprintStamScene.TacSprintStamScene")))))

(fn ConstructionScript)
