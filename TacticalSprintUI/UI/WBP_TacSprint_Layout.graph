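# Base HUD layout for the tactical sprint stamina meter, with named slots and empty SetStaminaProgress and OnHighlightActivated hooks.
# Energy component events drive progress, show the meter below full, pulse it when low and fade it out when full.
# Vehicles, riding, creative quickbar and DBNO hide the widget, while a tag query forces the meter visible.
# Events rebind on pawn change; layout adjusts for BR UI and overshield visibility.
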WidgetBlueprint /TacticalSprintUI/UI/WBP_TacSprint_Layout
animations [
 A_FullBarAnim {
  binding_count=1
  length_seconds=0.75
  object=&self:A_FullBarAnim
  sequence {bindings [Overlay_Energy {
     id=#1
     possessed_class=Overlay
     tracks [{
       type=MovieSceneFloatTrack
       display_name=Render Opacity
       sections [{type=MovieSceneFloatSection, end_frame=15, start_frame=0, channels [None {type=MovieSceneScriptingFloatChannel, keys [[0,1],[10,1],[15,0]]}]}]
      }]
    }]}
 }
 A_LowAnim {
  binding_count=1
  length_seconds=0.5
  object=&self:A_LowAnim
  sequence {bindings [Overlay_Energy {
     id=#2
     possessed_class=Overlay
     tracks [{
       type=MovieSceneFloatTrack
       display_name=Render Opacity
       sections [{
         type=MovieSceneFloatSection
         end_frame=10
         start_frame=0
         channels [None {type=MovieSceneScriptingFloatChannel, keys [[0,1],[5,0.33],[10,1]]}]
        }]
      }]
    }]}
 }
]
graphs [
 AreDBNOEnhancementsEnabled {nodes [
   n1 = AreDBNOEnhancementsEnabled!() -> (then, Pawn:Fort Pawn Object Reference) {is=title K2Node_FunctionEntry}
   n2 = GetComponentByClass!(self:Actor Object Reference=@n1, ComponentClass:Actor Component Class Reference=/Script/FortniteGame.DBNOBehaviorOverrideComponent) -> (ReturnValue:DBNOBehavior Override Component Object Reference)
   > n3 = Is Valid(InputObject=@n2)
   Return Node!(execute=@n3.Is Not Valid, Enabled:bool=false)
   > Return Node(Enabled=AreDBNOEnhancementsEnabled!(self:DBNOBehavior Override Component Object Reference=@n2) -> (ReturnValue:bool) {is=title K2Node_CallFunction})
  ]}
 BindToPlayerStateZoneEvents {nodes [
   BindToPlayerStateZoneEvents!() -> (then, DoBind:bool=~) {is=title K2Node_FunctionEntry}
   > n1 = Set PlayerStateZone(PlayerStateZone:Fort Player State Zone Object Reference=GetPlayerStateZone!(self:Fort HUDContext Object Reference=Get FortHUDContext!() -> (ReturnValue:Fort HUDContext Object Reference) {is=title K2Node_GetBlueprintContextEx}) -> (ReturnValue:Fort Player State Zone Object Reference)) -> (Output_Get:Fort Player State Zone Object Reference)
   > Is Valid(InputObject=@n1)
   > n2 = Sequence()
   > n3 = Branch(Condition=Get DoBind() -> (DoBind:bool))
   n4 = Create Event()
   n5 = Get PlayerStateZone() -> (PlayerStateZone:Fort Player State Zone Object Reference)
   > Bind Event to On Overshield Visibility Changed(self:Fort Player State Zone Object Reference=@n5, Delegate=@n4) {is=title K2Node_AddDelegate}
   > CheckOvershieldVisiblity()
   Unbind Event from On Overshield Visibility Changed(execute=@n3.else, self:Fort Player State Zone Object Reference=@n5, Delegate=@n4) {is=title K2Node_RemoveDelegate}
   n6 = GetPawnFromPlayerStateZone!(execute=@n2.then_1, InActor:Actor Object Reference=@n1) -> (then, OutPawn:Fort Player Pawn Object Reference, PawnReturned:bool)
   > Branch(Condition=@n6.PawnReturned)
   > n7 = Branch(Condition=AreDBNOEnhancementsEnabled!(self:Self Object Reference, Pawn:Fort Pawn Object Reference=@n6) -> (Enabled:bool))
   n8 = Branch(execute=@n7.else, Condition=Get DoBind())
   n9 = Create Event()
   > Bind Event to On DBNOState Changed(self:Fort Pawn Object Reference=@n6, Delegate=@n9) {is=title K2Node_AddDelegate}
   Unbind Event from On DBNOState Changed(execute=@n8.else, self:Fort Pawn Object Reference=@n6, Delegate=@n9) {is=title K2Node_RemoveDelegate}
  ]}
 BindToTacticalSprintComponent {nodes [
   n1 = BindToTacticalSprintComponent!() -> (then, DoBind:bool=~) {is=title K2Node_FunctionEntry}
   n2 = GetOwningPlayerPawn!(self:User Widget Object Reference) -> (ReturnValue:Pawn Object Reference)
   > Is Valid(InputObject=@n2)
   > n3 = Set TacticalSprintComponent(TacticalSprintComponent:Tactical Sprint Energy Component Object Reference=GetComponentByClass(self=@n2, ComponentClass=/TacticalSprintGame/Gameplay/TacticalSprintEnergyComponent.TacticalSprintEnergyComponent_C) -> (ReturnValue:Tactical Sprint Energy Component Object Reference)) -> (Output_Get:Tactical Sprint Energy Component Object Reference)
   > Is Valid(InputObject=@n3)
   > n4 = Branch(Condition=@n1)
   n5 = Create Event()
   n6 = Get TacticalSprintComponent() -> (TacticalSprintComponent:Tactical Sprint Energy Component Object Reference)
   > Bind Event to On Current Energy Changed(self:Fort Component Energy Object Reference=@n6, Delegate=@n5) {is=title K2Node_AddDelegate}
   n7 = Create Event()
   > Bind Event to On Energy Chunk Added(self:Fort Component Energy Object Reference=@n6, Delegate=@n7) {is=title K2Node_AddDelegate}
   > OnCurrentEnergyChanged!(execute, self, EnergyComponent:Fort Component Energy Object Reference=@n6, PreviousValue:float=0) -> (then)
   Unbind Event from On Current Energy Changed(execute=@n4.else, self:Fort Component Energy Object Reference=@n6, Delegate=@n5) {is=title K2Node_RemoveDelegate}
   > Unbind Event from On Energy Chunk Added(self:Fort Component Energy Object Reference=@n6, Delegate=@n7) {is=title K2Node_RemoveDelegate}
  ]}
 BindToVehicleEvents {nodes [
   n1 = BindToVehicleEvents!() -> (then, DoBind:bool=~) {is=title K2Node_FunctionEntry}
   > n2 = Branch(Condition=@n1)
   n3 = Create Event()
   n4 = Get FortGlobalUIContext() -> (ReturnValue:Fort Global UIContext Object Reference) {is=title K2Node_GetSubsystem}
   > Bind Event to On Exit Vehicle(self:Fort Global UIContext Object Reference=@n4, Delegate=@n3) {is=title K2Node_AddDelegate}
   n5 = Create Event()
   > Bind Event to On Enter Vehicle Driver(self:Fort Global UIContext Object Reference=@n4, Delegate=@n5) {is=title K2Node_AddDelegate}
   n6 = Create Event()
   > Bind Event to On Enter Vehicle Passenger(self:Fort Global UIContext Object Reference=@n4, Delegate=@n6) {is=title K2Node_AddDelegate}
   Unbind Event from On Exit Vehicle(execute=@n2.else, self:Fort Global UIContext Object Reference=@n4, Delegate=@n3) {is=title K2Node_RemoveDelegate}
   > Unbind Event from On Enter Vehicle Driver(self:Fort Global UIContext Object Reference=@n4, Delegate=@n5) {is=title K2Node_RemoveDelegate}
   > Unbind Event from On Enter Vehicle Passenger(self:Fort Global UIContext Object Reference=@n4, Delegate=@n6) {is=title K2Node_RemoveDelegate}
  ]}
 CheckOvershieldVisiblity {nodes [
   CheckOvershieldVisiblity() {is=title K2Node_FunctionEntry}
   > n1 = IsBRGameUIEnabled!(execute) -> (then, IsEnabled:bool)
   > n2 = Branch(Condition=@n1)
   > ClearWidthOverride(self:Size Box Object Reference=Get SizeBox_Progress() -> (SizeBox_Progress:Size Box Object Reference))
   > SetRenderTranslation(self=Get Overlay_Energy() -> (Overlay_Energy:Overlay Object Reference), Translation=(X=0,Y=0))
   n3 = GetPlayerStateZone(self=Get FortHUDContext())
   Is Valid(exec=@n2.else, InputObject=@n3)
   > n4 = Branch(Condition=ShouldOvershieldBarBeVisible!(self:Fort Player State Zone Object Reference=@n3) -> (ReturnValue:bool))
   > SetWidthOverride(self=Get SizeBox_Progress(), InWidthOverride=334)
   > SetRenderTranslation(self=Get Overlay_Energy(), Translation=(X=0,Y=-5))
   SetWidthOverride(execute=@n4.else, self=Get SizeBox_Progress(), InWidthOverride=334)
   > SetRenderTranslation(self=Get Overlay_Energy(), Translation=(X=0,Y=12))
  ]}
 EventGraph {nodes [
   n1 = OnEnergyAdded!() -> (
    OutputDelegate
    then
    EnergyComponent:Fort Component Energy Object Reference
    PreviousValue:float=~
    AdditionalTags:Gameplay Tag Container Structure
   ) {is=title K2Node_CustomEvent}
   n2 = Get HighlightGainTag() -> (HighlightGainTag:Gameplay Tag Structure)
   > Branch(Condition=AND Boolean(
     A=IsGameplayTagValid!(GameplayTag:Gameplay Tag Structure=@n2) -> (ReturnValue:bool)
     B=HasTag!(TagContainer:"Gameplay Tag Container Structure (by ref)"=@n1.AdditionalTags, Tag:Gameplay Tag Structure=@n2, bExactMatch:bool=true) -> (ReturnValue:bool)
    ))
   > OnHighlightActivated!(execute, self, EnergyComponent:Fort Component Energy Object Reference=@n1.EnergyComponent, PreviousValue:float=@n1.PreviousValue) -> (then)
   RemoveBinds() {is=title K2Node_CustomEvent}
   > BindToTacticalSprintComponent!(execute, self, DoBind:bool=false) -> (then) {is=title K2Node_CallFunction}
   > BindToPlayerStateZoneEvents!(execute, self, DoBind:bool=false) -> (then) {is=title K2Node_CallFunction}
   > BindToVehicleEvents!(execute, self, DoBind:bool=false) -> (then) {is=title K2Node_CallFunction}
   > Unbind Event from On Creative Quickbar Enabled Changed(
    self:Creative HUDLContext Object Reference=Get CreativeHUDLContext!() -> (ReturnValue:Creative HUDLContext Object Reference) {is=title K2Node_GetBlueprintContextEx}
    Delegate=Create Event()
   ) {is=title K2Node_RemoveDelegate}
   OnPawnSet() {is=title K2Node_CustomEvent}
   > RemoveBinds() {is=title K2Node_CallFunction}
   > SetupEventBinds()
   SetupEventBinds() {is=title K2Node_CustomEvent}
   > BindToTacticalSprintComponent(DoBind=true)
   > BindToPlayerStateZoneEvents(DoBind=true)
   > BindToVehicleEvents(DoBind=true)
   > Bind Event to On Creative Quickbar Equipped Changed(self:Creative HUDLContext Object Reference=Get CreativeHUDLContext(), Delegate=Create Event()) {is=title K2Node_AddDelegate}
   > UpdateVisibilityFromState()
   n3 = GetOwningPlayerPawn()
   > Is Valid(InputObject=@n3)
   > n4 = WaitGameplayTagAddToActor!(
    execute:Exec
    TargetActor:Actor Object Reference=@n3
    Tag:Gameplay Tag Structure=(TagName="Granted.Riding.Player.IsRiding")
    OnlyTriggerOnce:bool=false
   ) -> (then:Exec, AsyncTaskProxy:Ability Async Wait Gameplay Tag Added Object Reference, Added:Exec) {is=title K2Node_AsyncAction}
   > n5 = WaitGameplayTagRemoveFromActor!(
    execute:Exec
    TargetActor:Actor Object Reference=@n3
    Tag:Gameplay Tag Structure=(TagName="Granted.Riding.Player.IsRiding")
    OnlyTriggerOnce:bool=false
   ) -> (then:Exec, AsyncTaskProxy:Ability Async Wait Gameplay Tag Removed Object Reference, Removed:Exec) {is=title K2Node_AsyncAction}
   > n6 = WaitGameplayTagQueryOnActor!(
    execute:Exec
    TargetActor:Actor Object Reference=@n3
    TagQuery:Gameplay Tag Query Structure=Get ForceVisibleGameplayTagQuery() -> (ForceVisibleGameplayTagQuery:Gameplay Tag Query Structure)
    TriggerCondition:EWaitGameplayTagQueryTriggerCondition Enum=WhenTrue
    bOnlyTriggerOnce:bool=false
   ) -> (then:Exec, AsyncTaskProxy:Ability Async Wait Gameplay Tag Query Object Reference, Triggered:Exec) {is=title K2Node_AsyncAction}
   > n7 = WaitGameplayTagQueryOnActor(TargetActor=@n3, TagQuery=Get ForceVisibleGameplayTagQuery(), TriggerCondition=WhenFalse, bOnlyTriggerOnce=false)
   SetIsForceVisible!(execute=@n7.Triggered, self, NewIsForceVisible:bool=false) -> (then)
   SetIsForceVisible(execute=@n6.Triggered, NewIsForceVisible=true)
   Set IsRiding(execute=@n5.Removed, IsRiding:bool=false) -> (Output_Get:bool)
   > UpdateVisibilityFromState()
   Set IsRiding(execute=@n4.Added, IsRiding=true)
   > UpdateVisibilityFromState()
   n8 = OnCurrentEnergyChanged!() -> (OutputDelegate, then, EnergyComponent:Fort Component Energy Object Reference, PreviousValue:float=~) {is=title K2Node_CustomEvent}
   > n9 = Set CachedCurrentEnergy(CachedCurrentEnergy:double=GetCurrentEnergyPercentage!(self:Fort Component Energy Object Reference=@n8.EnergyComponent) -> (ReturnValue:float)) -> (Output_Get:double)
   > SetStaminaProgress!(execute, self, CurrentEnergy:double=@n9) -> (then)
   > n10 = Branch(Condition="float >= float"(A=@n9, B=1))
   > HideSprintMeter()
   ShowSprintMeter(execute=@n10.else)
   > Branch(Condition="float <= float"(A=Get CachedCurrentEnergy() -> (CachedCurrentEnergy:double), B=Get Low Energy Threshold() -> (Low Energy Threshold:double)) -> (ReturnValue=~))
   > TryPlayingLowState()
   ShowSprintMeter() {is=title K2Node_CustomEvent}
   > Set IsSprintMeterVisibile(IsSprintMeterVisibile:bool=true) -> (Output_Get:bool)
   > n11 = Sequence()
   > Branch(Condition=IsAnimationPlaying(InAnimation=Get A_FullBarAnim() -> (A_FullBarAnim:Widget Animation Object Reference)))
   > StopAnimation(InAnimation=Get A_FullBarAnim())
   SetVisibility(execute=@n11.then_1, self=Get Overlay_Energy(), InVisibility=HitTestInvisible)
   > n12 = Branch(Condition=IsAnimationPlaying(InAnimation=Get A_LowAnim() -> (A_LowAnim:Widget Animation Object Reference)))
   SetRenderOpacity(execute=@n12.else, self=Get Overlay_Energy(), InOpacity=1)
   HideSprintMeter() {is=title K2Node_CustomEvent}
   n13 = Get IsForceVisible() -> (IsForceVisible:bool)
   n14 = IsAnimationPlaying(InAnimation=Get A_FullBarAnim())
   n15 = Get CachedCurrentEnergy()
   > Branch(Condition=AND Boolean!(A=NOT Boolean(A=@n13), B="float >= float"(A=@n15, B=1), C:bool=NOT Boolean(A=@n14)) -> (ReturnValue))
   > Set IsSprintMeterVisibile(IsSprintMeterVisibile=false)
   > n16 = "Play Animation with Finished event (legacy)"!(
    execute:Exec
    Widget:User Widget Object Reference=Self-Reference()
    InAnimation:Widget Animation Object Reference=Get A_FullBarAnim()
    StartAtTime:float=0
    NumLoopsToPlay:int=1
    PlayMode:EUMGSequencePlayMode Enum
    PlaybackSpeed:float=1
   ) -> (then:Exec, Result:UMGSequence Player Object Reference, Finished:Exec) {is=title K2Node_PlayAnimation}
   n17 = Branch(execute=@n16.Finished, Condition=Get IsSprintMeterVisibile() -> (IsSprintMeterVisibile:bool))
   SetVisibility(execute=@n17.else, self=Get Overlay_Energy(), InVisibility=Hidden)
   Event Destruct()
   > RemoveBinds()
   n18 = Create Event()
   n19 = Get FortHUDContext()
   > Unbind Event from On Pawn Set(self:Fort HUDContext Object Reference=@n19, Delegate=@n18) {is=title K2Node_RemoveDelegate}
   Event Construct()
   > SetVisibility(self=Get Overlay_Energy(), InVisibility=Hidden)
   > Bind Event to On Pawn Set(self:Fort HUDContext Object Reference=@n19, Delegate=@n18) {is=title K2Node_AddDelegate}
   > SetupEventBinds() {is=title K2Node_CallFunction}
   > SetSize()
  ]}
 GetIsDBNO {nodes [
   GetIsDBNO() {is=title K2Node_FunctionEntry}
   > n1 = GetPawnFromPlayerStateZone(InActor=GetPlayerStateZone(self=Get FortHUDContext()))
   > n2 = Branch(Condition=@n1.PawnReturned)
   n3 = AreDBNOEnhancementsEnabled(Pawn=@n1)
   > Return Node!(
    execute
    IsDBNO:bool=AND Boolean!(A=IsDBNO!(self:Fort Pawn Object Reference=@n1) -> (ReturnValue:bool), B=NOT Boolean(A=@n3)) -> (ReturnValue)
   )
   Return Node(execute=@n2.else, IsDBNO=false)
  ]}
 GetIsInVehicle {nodes [
   GetIsInVehicle() {is=title K2Node_FunctionEntry}
   > n1 = Cast To FortPlayerController(Object=GetOwningPlayer())
   n2 = GetPlayerPawn!(self:Fort Player Controller Object Reference=@n1) -> (ReturnValue:Fort Player Pawn Object Reference)
   > n3 = Is Valid(InputObject=@n2)
   Return Node!(execute=[@n3.Is Not Valid, @n1.CastFailed], IsInVehicle:bool=false)
   > Return Node(IsInVehicle=IsInVehicle!(self:Fort Player Pawn Object Reference=@n2) -> (ReturnValue:bool))
  ]}
 GetIsQuickbarEquipped {nodes [
   GetIsQuickbarEquipped() {is=title K2Node_FunctionEntry}
   > n1 = Cast To FortPlayerControllerGameplay!(execute, Object=GetOwningPlayer()) -> (then, CastFailed, AsFort Player Controller Gameplay:Fort Player Controller Gameplay Object Reference) {is=title K2Node_DynamicCast}
   > Return Node(IsInVehicle=IsCreativeQuickbarActive!(self:Fort Player Controller Gameplay Object Reference=@n1) -> (ReturnValue:bool))
   Return Node(execute=@n1.CastFailed, IsInVehicle=false)
  ]}
 HandleEnterVehicle {nodes [
   HandleEnterVehicle() {is=title K2Node_FunctionEntry}
   > UpdateVisibilityFromState()
  ]}
 HandleExitVehicle {nodes [
   HandleExitVehicle() {is=title K2Node_FunctionEntry}
   > UpdateVisibilityFromState()
  ]}
 OnCreativeQuickbarEquippedChanged {nodes [
   OnCreativeQuickbarEquippedChanged!() -> (then, bIsQuickbarEquipped:bool=~) {is=title K2Node_FunctionEntry}
   > UpdateVisibilityFromState()
  ]}
 OnDBNOStateChanged {nodes [
   n1 = OnDBNOStateChanged!() -> (then, FortPawn:Fort Pawn Object Reference, bInIsDBNO:bool=~) {is=title K2Node_FunctionEntry}
   > Branch(Condition=NOT Boolean(A=AreDBNOEnhancementsEnabled(Pawn=@n1)))
   > SetVisibility(InVisibility=Select(Option 0=HitTestInvisible, Option 1=Hidden, Index=@n1.bInIsDBNO))
  ]}
 OnHighlightActivated {nodes [OnHighlightActivated!() -> (then, EnergyComponent:Fort Component Energy Object Reference, PreviousValue:float=~) {is=title K2Node_FunctionEntry}]}
 OnOvershieldVisibilityChanged {nodes [
   OnOvershieldVisibilityChanged!() -> (then, NewVisible:bool=~) {is=title K2Node_FunctionEntry}
   > CheckOvershieldVisiblity() {is=title K2Node_CallFunction}
  ]}
 SetIsForceVisible {nodes [
   n1 = SetIsForceVisible!() -> (then, NewIsForceVisible:bool=~) {is=title K2Node_FunctionEntry}
   > n2 = Set IsForceVisible(IsForceVisible:bool=@n1) -> (Output_Get:bool)
   > n3 = Branch(Condition=@n2)
   > ShowSprintMeter() {is=title K2Node_CallFunction}
   HideSprintMeter(execute=@n3.else) {is=title K2Node_CallFunction}
  ]}
 SetSize {nodes [
   SetSize() {is=title K2Node_FunctionEntry}
   > n1 = IsBRGameUIEnabled()
   > n2 = Branch(Condition=@n1)
   n3 = Get SizeBox_Progress()
   > ClearWidthOverride(self=@n3)
   > ClearHeightOverride(self:Size Box Object Reference=@n3)
   > SetPadding!(
    execute
    self:Horizontal Box Slot Object Reference=SlotAsHorizontalBoxSlot!(Widget:Widget Object Reference=@n3) -> (ReturnValue:Horizontal Box Slot Object Reference)
    InPadding_Left:float=0
    InPadding_Top:float=0
    InPadding_Right:float=0
    InPadding_Bottom:float=0
   ) -> (then)
   > SetVisibility(
    self=[
     Get SizeBox_Icon() -> (SizeBox_Icon:Size Box Object Reference)
     Get NamedSlot_AdditionalContent() -> (NamedSlot_AdditionalContent:Named Slot Object Reference)
    ]
    InVisibility=Collapsed
   )
   n4 = Get SizeBox_Progress()
   SetHeightOverride(execute=@n2.else, self=@n4, InHeightOverride=6)
   > SetWidthOverride(self=@n4, InWidthOverride=334)
   > SetPadding(self=SlotAsHorizontalBoxSlot(Widget=@n4), InPadding_Left=0, InPadding_Top=0, InPadding_Right=0, InPadding_Bottom=6)
   > SetPadding(
    self:Overlay Slot Object Reference=SlotAsOverlaySlot(Widget=Get HorizontalBox_Energy() -> (HorizontalBox_Energy:Horizontal Box Object Reference))
    InPadding_Left=0
    InPadding_Top=0
    InPadding_Right=0
    InPadding_Bottom=0
   )
   > SetVisibility(self=[Get SizeBox_Icon(), Get NamedSlot_AdditionalContent()], InVisibility=HitTestInvisible)
  ]}
 SetStaminaProgress {nodes [SetStaminaProgress!() -> (then, CurrentEnergy:double=~) {is=title K2Node_FunctionEntry}]}
 TryPlayingLowState {nodes [
   TryPlayingLowState() {is=title K2Node_FunctionEntry}
   > n1 = Branch(Condition=IsAnimationPlaying(InAnimation=Get A_LowAnim()))
   PlayAnimation(execute=@n1.else, InAnimation=Get A_LowAnim())
  ]}
 UpdateVisibilityFromState {nodes [
   UpdateVisibilityFromState() {is=title K2Node_FunctionEntry}
   n1 = GetIsInVehicle!(self) -> (IsInVehicle:bool) {is=title K2Node_CallFunction}
   n2 = GetIsQuickbarEquipped!(self) -> (IsInVehicle:bool) {is=title K2Node_CallFunction}
   n3 = GetIsDBNO!(self) -> (IsDBNO:bool) {is=title K2Node_CallFunction}
   n4 = Get IsRiding() -> (IsRiding:bool)
   > SetVisibility(InVisibility=Select(Option 0=SelfHitTestInvisible, Option 1=Hidden, Index=OR Boolean!(A=@n4, B=@n1, C:bool=@n2, D:bool=@n3) -> (ReturnValue)))
  ]}
]
variables [
 CachedCurrentEnergy {type=number}
 ForceVisibleGameplayTagQuery {
  type=GameplayTagQuery
  default {
   autoDescription=" ANY( Granted.Athena.Energy.ConstantRegen )"
   tokenStreamVersion=0
   userDescription=""
   queryTokenStream [0,1,1,1,0]
   tagDictionary [{tagName=Granted.Athena.Energy.ConstantRegen}]
  }
 }
 HighlightGainTag {type=GameplayTag, default {tagName=HUD.Athena.TacticalSprint.HighlightGain}}
 IsForceVisible {type=boolean}
 IsRiding {type=boolean}
 IsSprintMeterVisibile {type=boolean}
 Low Energy Threshold {type=number, default=0.25}
]
widgets {
 info {inheritedWidgetCount=1, parentClass=&FortHUDElementWidget, rootWidgetClass=&Overlay, widgetCount=8}
 widgets [
  Overlay_Energy {widgetClassPath=&Overlay, bIsVariable=true, widget=Overlay_Energy}
  HorizontalBox_Energy {
   widgetClassPath=&HorizontalBox
   bIsVariable=true
   parent=Overlay_Energy
   widget=HorizontalBox_Energy
   slot {class=OverlaySlot, properties {horizontalAlignment=HAlign_Fill, verticalAlignment=VAlign_Fill}}
  }
  SizeBox_Icon {
   widgetClassPath=&SizeBox
   bIsVariable=true
   parent=HorizontalBox_Energy
   widget=SizeBox_Icon
   properties {bOverride_HeightOverride=true, bOverride_WidthOverride=true, heightOverride=28, widthOverride=28}
   slot {class=HorizontalBoxSlot, properties {padding=<ltrb 0 0 9 0>}}
  }
  NamedSlot_Icon {widgetClassPath=&NamedSlot, parent=SizeBox_Icon, widget=NamedSlot_Icon, slot {class=SizeBoxSlot}}
  SizeBox_Progress {
   widgetClassPath=&SizeBox
   bIsVariable=true
   parent=HorizontalBox_Energy
   widget=SizeBox_Progress
   properties {bOverride_HeightOverride=true, bOverride_WidthOverride=true, heightOverride=6, widthOverride=345}
   slot {class=HorizontalBoxSlot, properties {verticalAlignment=VAlign_Bottom, padding=<ltrb 0 0 0 6>, size {sizeRule=Fill, value=1}}}
  }
  NamedSlot_ProgressBar {widgetClassPath=&NamedSlot, parent=SizeBox_Progress, widget=NamedSlot_ProgressBar, slot {class=SizeBoxSlot}}
  NamedSlot_AdditionalContent {
   widgetClassPath=&NamedSlot
   bIsVariable=true
   parent=HorizontalBox_Energy
   widget=NamedSlot_AdditionalContent
   slot {class=HorizontalBoxSlot}
  }
  TutorialHighlight {widgetClassPath=&TutorialHighlight}
 ]
}
