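# Tactical sprint stamina meter widget, derived from WBP_TacSprint_Layout, that shows energy and dims the sprint icon when empty.
# Stacked fill, delta and highlight images are driven through the Progress parameter of cached dynamic materials.
# Drops beyond a threshold briefly show the previous value on the delta bar.
# Highlighted gains hold the fill back while the highlight shows current energy, with timers handling the delayed catch-up.
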WidgetBlueprint /TacticalSprintUI/UI/WBP_TacSprint_Meter
graphs [
 CacheMaterials {nodes [
   CacheMaterials() {is=title K2Node_FunctionEntry}
   > n1 = GetDynamicMaterial(self=Get DeltaBar() -> (DeltaBar:Image Object Reference))
   > Set DeltaBarMID(DeltaBarMID:Material Instance Dynamic Object Reference=@n1) -> (Output_Get:Material Instance Dynamic Object Reference)
   > n2 = GetDynamicMaterial(self=Get FillBar() -> (FillBar:Image Object Reference))
   > Set FillBarMID(FillBarMID:Material Instance Dynamic Object Reference=@n2) -> (Output_Get:Material Instance Dynamic Object Reference)
   > n3 = GetDynamicMaterial(self=Get HighlightBar() -> (HighlightBar:Image Object Reference))
   > n4 = Set HighlightBarMID(HighlightBarMID:Material Instance Dynamic Object Reference=@n3) -> (Output_Get:Material Instance Dynamic Object Reference)
   > SetScalarParameterValue(self=@n4, ParameterName=Progress, Value=0)
  ]}
 DisableHighlightBar {nodes [
   DisableHighlightBar() {is=title K2Node_FunctionEntry}
   > Set IsGainHighlightActive(IsGainHighlightActive:bool=false) -> (Output_Get:bool)
   > Clear and Invalidate Timer by Handle(Handle=Get EmptyHighlight_Handle() -> (EmptyHighlight_Handle:Timer Handle Structure))
   > Clear and Invalidate Timer by Handle(Handle=Get StartHighlightFadeOut_Handle() -> (StartHighlightFadeOut_Handle:Timer Handle Structure))
   > SetScalarParameterValue(self=Get HighlightBarMID() -> (HighlightBarMID:Material Instance Dynamic Object Reference), ParameterName=Progress, Value=0)
   > SetVisibility(self=Get HighlightBar(), InVisibility=Collapsed)
  ]}
 EmptyDelta {nodes [
   EmptyDelta() {is=title K2Node_FunctionEntry}
   > SetScalarParameterValue(self=Get DeltaBarMID() -> (DeltaBarMID:Material Instance Dynamic Object Reference), ParameterName=Progress, Value=0)
  ]}
 EmptyHighlightBar {nodes [
   EmptyHighlightBar() {is=title K2Node_FunctionEntry}
   > n1 = Branch(Condition="Not Equal (Float)"(A=Get ValueBeforeHighlightedGain() -> (ValueBeforeHighlightedGain:double), B=Get PrevEnergy() -> (PrevEnergy:double)))
   n2 = Get EmptyHighlight_StartTime() -> (EmptyHighlight_StartTime:double)
   n3 = Get EmptyHighlight_TotalSeconds() -> (EmptyHighlight_TotalSeconds:double)
   > Set ValueBeforeHighlightedGain(ValueBeforeHighlightedGain:double=MapRangeClamped(
     Value=GetTimeSeconds!() -> (ReturnValue:double)
     InRangeA=Get EmptyHighlight_StartTime()
     InRangeB=float + float(A=@n2, B=@n3)
     OutRangeA=Get ValueBeforeHighlightedGain()
     OutRangeB=Get PrevEnergy()
    )) -> (Output_Get:double)
   > SetStaminaProgress!(execute, self:WBP Tac Sprint Layout Object Reference, CurrentEnergy:double=Get PrevEnergy()) -> (then)
   DisableHighlightBar(execute=@n1.else) {is=title K2Node_CallFunction}
  ]}
 EventGraph {nodes [
   Event Destruct()
   > Parent: Destruct()
   > DisableHighlightBar()
   Event Construct()
   > Parent: Construct()
   Event OnInitialized()
   > CacheMaterials() {is=title K2Node_CallFunction}
  ]}
 FadeOutHighlightBar {nodes [
   FadeOutHighlightBar() {is=title K2Node_FunctionEntry}
   > Clear and Invalidate Timer by Handle(Handle=Get StartHighlightFadeOut_Handle())
   > Clear and Invalidate Timer by Handle(Handle=Get EmptyHighlight_Handle())
   > n1 = Set Timer by Event(Delegate=Create Event(), Time=0.03, bLooping=true, bMaxOncePerFrame=true, InitialStartDelay=0.03)
   > Set EmptyHighlight_Handle(EmptyHighlight_Handle:Timer Handle Structure=@n1) -> (Output_Get:Timer Handle Structure)
   > Set EmptyHighlight_StartTime(EmptyHighlight_StartTime:double=GetTimeSeconds()) -> (Output_Get:double)
  ]}
 OnHighlightActivated {nodes [
   OnHighlightActivated!() -> (then, EnergyComponent:Fort Component Energy Object Reference, PreviousValue:float) {is=title K2Node_FunctionEntry}
   > Set IsGainHighlightActive(IsGainHighlightActive=true)
   > Set ValueBeforeHighlightedGain(ValueBeforeHighlightedGain=Get PrevEnergy())
   > n1 = Branch(Condition=Is Valid Timer Handle!(Handle:Timer Handle Structure=Get StartHighlightFadeOut_Handle()) -> (ReturnValue:bool))
   > n2 = Clear and Invalidate Timer by Handle(Handle=Get StartHighlightFadeOut_Handle())
   n3 = Set Timer by Event(execute=[@n2.then, @n1.else], Delegate=Create Event(), Time=Get Highlight_TotalSeconds() -> (Highlight_TotalSeconds:double), bLooping=false)
   > Set StartHighlightFadeOut_Handle(StartHighlightFadeOut_Handle:Timer Handle Structure=@n3) -> (Output_Get:Timer Handle Structure)
   > Set PrevEnergy(PrevEnergy:double=GetCurrentEnergyPercentage!(self:Fort Component Energy Object Reference=Get EnergyComponent() -> (EnergyComponent:Fort Component Energy Object Reference)) -> (ReturnValue:float)) -> (Output_Get:double)
   > SetVisibility(self=Get HighlightBar(), InVisibility=SelfHitTestInvisible)
   > Clear and Invalidate Timer by Handle(Handle=Get EmptyHighlight_Handle())
  ]}
 SetDelta {nodes [
   n1 = SetDelta!() -> (then, CurrentEnergy:double=~) {is=title K2Node_FunctionEntry}
   n2 = Get Anim_DeltaThreshold() -> (Anim_DeltaThreshold:double)
   n3 = Get PrevEnergy()
   > n4 = Branch(Condition="float <= float"(A=@n1, B=float - float(A=@n3, B=@n2)) -> (ReturnValue=~))
   > SetScalarParameterValue(self=Get DeltaBarMID(), ParameterName=Progress, Value=Get PrevEnergy())
   > Clear and Invalidate Timer by Handle(Handle=Get EmptyDelta_Handle() -> (EmptyDelta_Handle:Timer Handle Structure))
   > n5 = Set Timer by Event(Delegate=Create Event(), Time=0.33, bLooping=false)
   > Set EmptyDelta_Handle(EmptyDelta_Handle:Timer Handle Structure=@n5) -> (Output_Get:Timer Handle Structure)
   n6 = Branch(execute=@n4.else, Condition=Is Timer Active by Handle!(Handle:Timer Handle Structure=Get EmptyDelta_Handle()) -> (ReturnValue:bool))
   EmptyDelta(execute=@n6.else) {is=title K2Node_CallFunction}
  ]}
 SetIsForceVisible {nodes [
   n1 = SetIsForceVisible!() -> (then, NewIsForceVisible:bool) {is=title K2Node_FunctionEntry}
   > Parent: SetIsForceVisible!(execute, NewIsForceVisible:bool=@n1) -> (then) {is=title K2Node_CallParentFunction}
   > n2 = Branch(Condition=@n1)
   n3 = Get DeltaBar()
   n4 = Get FillBar()
   > SetVisibility(self=[@n3, @n4], InVisibility=Collapsed)
   SetVisibility(execute=@n2.else, self=[@n3, @n4], InVisibility=HitTestInvisible)
  ]}
 SetStaminaProgress {nodes [
   n1 = SetStaminaProgress!() -> (then, CurrentEnergy:double) {is=title K2Node_FunctionEntry}
   > Parent: SetStaminaProgress!(execute, CurrentEnergy:double=@n1) -> (then) {is=title K2Node_CallParentFunction}
   n2 = Get CurrentEnergy() -> (CurrentEnergy:double)
   > SetRenderOpacity(
    self=Get SprintIcon() -> (SprintIcon:Image Object Reference)
    InOpacity=Select(Option 0:double=1, Option 1:double=0.33, Index="float <= float"(A=@n2) -> (ReturnValue=~)) -> (ReturnValue:double)
   )
   > n3 = Branch(Condition=Get IsGainHighlightActive() -> (IsGainHighlightActive:bool))
   n4 = float - float(A=Get CurrentEnergy(), B=Get PrevEnergy())
   > n5 = Branch(Condition="float > float"(A=@n4))
   > n6 = Set ValueBeforeHighlightedGain(ValueBeforeHighlightedGain=float + float(A=@n4, B=Get ValueBeforeHighlightedGain()))
   n7 = Branch(execute=[@n6.then, @n5.else], Condition="float > float"(A=Get CurrentEnergy(), B=Get ValueBeforeHighlightedGain()))
   > SetScalarParameterValue(
    self=Get FillBarMID() -> (FillBarMID:Material Instance Dynamic Object Reference)
    ParameterName=Progress
    Value=Get ValueBeforeHighlightedGain()
   )
   > SetScalarParameterValue(self=Get HighlightBarMID(), ParameterName=Progress, Value=Get CurrentEnergy())
   > SetDelta!(execute, self, CurrentEnergy:double=Get ValueBeforeHighlightedGain()) -> (then) {is=title K2Node_CallFunction}
   > Set PrevEnergy(PrevEnergy=Get CurrentEnergy())
   DisableHighlightBar(execute=@n7.else)
   SetScalarParameterValue(execute=@n3.else, self=Get FillBarMID(), ParameterName=Progress, Value=Get CurrentEnergy())
   n8 = Get CurrentEnergy()
   > SetDelta(CurrentEnergy=@n8)
   > Set PrevEnergy(PrevEnergy=@n8)
  ]}
]
variables [
 Anim_DeltaThreshold {type=number, default=0.05}
 DeltaBarMID {type=MaterialInstanceDynamic}
 EmptyDelta_Handle {type=TimerHandle}
 EmptyHighlight_Handle {type=TimerHandle}
 EmptyHighlight_StartTime {type=number, default=0.3}
 EmptyHighlight_TotalSeconds {type=number, default=0.5}
 FillBarMID {type=MaterialInstanceDynamic}
 Highlight_TotalSeconds {type=number, default=1.5}
 HighlightBarMID {type=MaterialInstanceDynamic}
 IsGainHighlightActive {type=boolean}
 PrevEnergy {type=number}
 StartHighlightFadeOut_Handle {type=TimerHandle}
 ValueBeforeHighlightedGain {type=number}
]
widgets {
 info {
  inheritedWidgetCount=0
  namedSlotCount=2
  parentClass=&/WBP_TacSprint_Layout.WBP_TacSprint_Layout_C
  rootWidgetClass=null
  widgetCount=5
 }
 widgets [
  SprintIcon {
   widgetClassPath=&Image
   bIsVariable=true
   widget=SprintIcon
   properties {brush {
     bIsDynamicallyLoaded=false
     drawAs=Image
     imageType=NoImage
     mirroring=NoMirror
     resourceName=None
     resourceObject=/TacticalSprintUI/UI/Materials/M_UI_TacSprint_Icon
     tiling=NoTile
     imageSize=<32 32>
     margin=<ltrb 0.2 0.2 0.2 0.2>
     outlineSettings {
      bUseBrushTransparency=false
      roundingType=HalfHeightRadius
      width=0
      color {colorUseRule=UseColor_Specified, specifiedColor=<rgba 0 0 0 0>}
      cornerRadii=<0 0 0 0>
     }
     tintColor {colorUseRule=UseColor_Specified, specifiedColor=<rgba 1 1 1 1>}
     uVRegion {bIsValid=false, max=<0 0>, min=<0 0>}
    }}
  }
  Overlay_2 {widgetClassPath=&Overlay, widget=Overlay_2}
  DeltaBar {
   widgetClassPath=&Image
   bIsVariable=true
   parent=Overlay_2
   widget=DeltaBar
   properties {
    visibility=HitTestInvisible
    brush {
     bIsDynamicallyLoaded=false
     drawAs=Image
     imageType=NoImage
     mirroring=NoMirror
     resourceName=None
     resourceObject=/TacticalSprintUI/UI/Materials/M_UI_HitPointBar_EnergyDelta
     tiling=NoTile
     imageSize=<32 32>
     margin=<ltrb 0 0 0 0>
     outlineSettings {
      bUseBrushTransparency=false
      roundingType=HalfHeightRadius
      width=0
      color {colorUseRule=UseColor_Specified, specifiedColor=<rgba 0 0 0 0>}
      cornerRadii=<0 0 0 0>
     }
     tintColor {colorUseRule=UseColor_Specified, specifiedColor=<rgba 1 1 1 1>}
     uVRegion {bIsValid=false, max=<0 0>, min=<0 0>}
    }
   }
   slot {class=OverlaySlot, properties {horizontalAlignment=HAlign_Fill}}
  }
  HighlightBar {
   widgetClassPath=&Image
   bIsVariable=true
   parent=Overlay_2
   widget=HighlightBar
   properties {
    visibility=Collapsed
    brush {
     bIsDynamicallyLoaded=false
     drawAs=Image
     imageType=FullColor
     mirroring=NoMirror
     resourceName=None
     resourceObject=/TacticalSprintUI/UI/Materials/M_UI_HitPointBar_EnergyHighlight
     tiling=NoTile
     imageSize=<32 32>
     margin=<ltrb 0 0 0 0>
     outlineSettings {
      bUseBrushTransparency=false
      roundingType=HalfHeightRadius
      width=0
      color {colorUseRule=UseColor_Specified, specifiedColor=<rgba 0 0 0 0>}
      cornerRadii=<0 0 0 0>
     }
     tintColor {colorUseRule=UseColor_Specified, specifiedColor=<rgba 1 1 1 1>}
     uVRegion {bIsValid=false, max=<0 0>, min=<0 0>}
    }
   }
   slot {class=OverlaySlot, properties {horizontalAlignment=HAlign_Fill}}
  }
  FillBar {
   widgetClassPath=&Image
   bIsVariable=true
   parent=Overlay_2
   widget=FillBar
   properties {brush {
     bIsDynamicallyLoaded=false
     drawAs=Image
     imageType=NoImage
     mirroring=NoMirror
     resourceName=None
     resourceObject=/TacticalSprintUI/UI/Materials/M_UI_HitPointBar_Energy
     tiling=NoTile
     imageSize=<32 32>
     margin=<ltrb 0 0 0 0>
     outlineSettings {
      bUseBrushTransparency=false
      roundingType=HalfHeightRadius
      width=0
      color {colorUseRule=UseColor_Specified, specifiedColor=<rgba 0 0 0 0>}
      cornerRadii=<0 0 0 0>
     }
     tintColor {colorUseRule=UseColor_Specified, specifiedColor=<rgba 1 1 1 1>}
     uVRegion {bIsValid=false, max=<0 0>, min=<0 0>}
    }}
   slot {class=OverlaySlot, properties {horizontalAlignment=HAlign_Fill}}
  }
 ]
}
